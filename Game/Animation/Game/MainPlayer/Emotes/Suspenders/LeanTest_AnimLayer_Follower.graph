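# Test animation layer Blueprint for the Suspenders emote follower on the Fortnite player skeleton, adding procedural left/right lean from turning.
# Each update, Lean divides the owning actor's yaw change by delta time, then scales and clamps it into LeanAngle.
# LeanAngle drives additive lean blendspaces in AnimGraph and ItemFinalPoseOverride, while ItemFullBodyOverride holds X at zero.
# Debug prints remain, Lean2 and Lean3 are uncalled experiments, and most other layers are empty stubs.
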AnimBlueprint /Game/Animation/Game/MainPlayer/Emotes/Suspenders/LeanTest_AnimLayer_Follower
target_skeleton=&Fortnite_M_Avg_Player_Skeleton
graphs [
 AnimGraph {nodes [
   Blend Poses by bool(BlendPose_0=Slot 'InMotion'() {slotName=InMotion}, BlendPose_1=Slot 'DefaultSlot'(), BlendTime_0=0.3, BlendTime_1=0.3)
   AnimGraphNode_BlendSpaceGraph_0 = Blendspace 'BS_LeanTest'!(X:float=Get LeanAngle() -> (LeanAngle:double)) -> (Pose:Pose Link Structure) {is=title AnimGraphNode_BlendSpaceGraph}
   Output Pose(Result=Apply Additive(Additive=@AnimGraphNode_BlendSpaceGraph_0))
   PrintString!(execute, InString=Hello, bPrintToScreen, bPrintToLog, TextColor, Duration, Key) -> (then)
  ]}
 BS_LeanTest {path=AnimGraph.AnimGraphNode_BlendSpaceGraph_0.BS_LeanTest, links [], nodes []}
 Emote_Suspenders_CMF_LeanLeft {
  path=AnimGraph.AnimGraphNode_BlendSpaceGraph_0.BS_LeanTest.Emote_Suspenders_CMF_LeanLeft
  nodes [Output Animation Pose!(Result:Pose Link Structure="Sequence Player (additive) 'Emote_Suspenders_CMF_LeanLeft'"() {method=Graph, sequence=&/Emote_Suspenders_CMF_LeanLeft}) {is=title AnimGraphNode_BlendSpaceSampleResult}]
 }
 Emote_Suspenders_CMF_LeanRight {
  path=AnimGraph.AnimGraphNode_BlendSpaceGraph_0.BS_LeanTest.Emote_Suspenders_CMF_LeanRight
  nodes [Output Animation Pose(Result="Sequence Player (additive) 'Emote_Suspenders_CMF_LeanRight'"() {method=Graph, sequence=&/Emote_Suspenders_CMF_LeanRight})]
 }
 Emote_Suspenders_CMF_NoLean {
  path=AnimGraph.AnimGraphNode_BlendSpaceGraph_0.BS_LeanTest.Emote_Suspenders_CMF_NoLean
  nodes [Output Animation Pose(Result="Sequence Player (additive) 'Emote_Suspenders_CMF_NoLean'"() {method=Graph, sequence=&/Emote_Suspenders_CMF_NoLean})]
 }
 BlueprintThreadSafeUpdateAnimation {nodes [BlueprintThreadSafeUpdateAnimation!() -> (then, DeltaTime:float) {is=title K2Node_FunctionEntry}]}
 EventGraph {nodes [
   n1 = Event BlueprintUpdateAnimation()
   > Lean!(execute, self, OwningActor:Actor Object Reference=Get Actor() -> (Actor:Actor Object Reference), DeltaTime:double=@n1.DeltaTimeX) -> (then)
   Event BlueprintInitializeAnimation()
   > Set Actor(Actor:Actor Object Reference=GetOwningActor()) -> (Output_Get:Actor Object Reference)
   Event BlueprintBeginPlay() {is=title K2Node_Event}
   > PrintString(InString=play)
   TryGetPawnOwner()
  ]}
 ItemAimOffset {nodes [
   Input Pose!() -> (Pose, AimOffsetAlpha:double, Yaw:double, Pitch:double) {name=InPose_AimOffset}
   Input Pose!() -> (Pose) {name=InPose_PreAimOffsets}
   Output Pose() {layerGroup=Item Override}
  ]}
 ItemFallAdditive {nodes [
   Input Pose() {name=InPoseItemFallAdditive}
   Output Pose() {layerGroup=Item Override}
  ]}
 ItemFallLandAdditive {nodes [
   Input Pose() {name=InPoseItemFallLandAdditive}
   Output Pose() {layerGroup=Item Override}
  ]}
 ItemFallLevel2Override {nodes [
   Input Pose() {name=InPose_FallLevel2}
   Output Pose() {layerGroup=Item Override}
  ]}
 ItemFinalPoseOverride {nodes [
   AnimGraphNode_BlendSpaceGraph_0 = Blendspace 'BS_LeanTest'(X=Get LeanAngle())
   n1 = Input Pose() {name=InFinalPose}
   Output Pose(Result=LogString!(
     Source:Pose Link Structure=Apply Additive(Base=@n1, Additive=@AnimGraphNode_BlendSpaceGraph_0)
     InString:String=finalpose
     bPrintToLog:bool=true
    ) -> (Pose:Pose Link Structure) {is=title AnimGraphNode_CallFunction}) {layerGroup=Item Override}
  ]}
 BS_LeanTest {path=ItemFinalPoseOverride.AnimGraphNode_BlendSpaceGraph_0.BS_LeanTest, links [], nodes []}
 Emote_Suspenders_CMF_LeanLeft {
  path=ItemFinalPoseOverride.AnimGraphNode_BlendSpaceGraph_0.BS_LeanTest.Emote_Suspenders_CMF_LeanLeft
  nodes [Output Animation Pose(Result="Sequence Player (additive) 'Emote_Suspenders_CMF_LeanLeft'"() {method=Graph, sequence=&Emote_Suspenders_CMF_LeanLeft})]
 }
 Emote_Suspenders_CMF_LeanRight {
  path=ItemFinalPoseOverride.AnimGraphNode_BlendSpaceGraph_0.BS_LeanTest.Emote_Suspenders_CMF_LeanRight
  nodes [Output Animation Pose(Result="Sequence Player (additive) 'Emote_Suspenders_CMF_LeanRight'"() {method=Graph, sequence=&Emote_Suspenders_CMF_LeanRight})]
 }
 Emote_Suspenders_CMF_NoLean {
  path=ItemFinalPoseOverride.AnimGraphNode_BlendSpaceGraph_0.BS_LeanTest.Emote_Suspenders_CMF_NoLean
  nodes [Output Animation Pose(Result="Sequence Player (additive) 'Emote_Suspenders_CMF_NoLean'"() {method=Graph, sequence=&Emote_Suspenders_CMF_NoLean})]
 }
 ItemFullBodyAimOffsets {nodes [
   Input Pose!() -> (Pose, FullBody_AimOffsetAlpha:double, Aim_Yaw:double, Aim_Pitch:double) {name=InPoseFullBodyAimOffsets}
   Input Pose!() -> (Pose) {name=InPosePreFullBodyAimOffsets}
   Output Pose() {layerGroup=Item Override}
  ]}
 ItemFullBodyOverride {nodes [
   Blend Poses by bool(BlendPose_0=Slot 'InMotion'() {slotName=InMotion}, BlendPose_1=Slot 'DefaultSlot'(), BlendTime_0=0.3, BlendTime_1=0.3)
   AnimGraphNode_BlendSpaceGraph_0 = Blendspace 'BS_LeanTest'(X=0)
   n1 = Input Pose() {name=InPoseFullBody}
   Output Pose(Result=LogString(Source=Apply Additive(Base=@n1, Additive=@AnimGraphNode_BlendSpaceGraph_0), InString=finalpose, bPrintToLog=true)) {layerGroup=Item Override}
  ]}
 BS_LeanTest {path=ItemFullBodyOverride.AnimGraphNode_BlendSpaceGraph_0.BS_LeanTest, links [], nodes []}
 Emote_Suspenders_CMF_LeanLeft {
  path=ItemFullBodyOverride.AnimGraphNode_BlendSpaceGraph_0.BS_LeanTest.Emote_Suspenders_CMF_LeanLeft
  nodes [Output Animation Pose(Result="Sequence Player (additive) 'Emote_Suspenders_CMF_LeanLeft'"() {method=Graph, sequence=&Emote_Suspenders_CMF_LeanLeft})]
 }
 Emote_Suspenders_CMF_LeanRight {
  path=ItemFullBodyOverride.AnimGraphNode_BlendSpaceGraph_0.BS_LeanTest.Emote_Suspenders_CMF_LeanRight
  nodes [Output Animation Pose(Result="Sequence Player (additive) 'Emote_Suspenders_CMF_LeanRight'"() {method=Graph, sequence=&Emote_Suspenders_CMF_LeanRight})]
 }
 Emote_Suspenders_CMF_NoLean {
  path=ItemFullBodyOverride.AnimGraphNode_BlendSpaceGraph_0.BS_LeanTest.Emote_Suspenders_CMF_NoLean
  nodes [Output Animation Pose(Result="Sequence Player (additive) 'Emote_Suspenders_CMF_NoLean'"() {method=Graph, sequence=&Emote_Suspenders_CMF_NoLean})]
 }
 ItemIdleOverride {nodes [
   Input Pose() {name=InPose_Idle}
   Output Pose() {layerGroup=Item Override}
  ]}
 ItemJetPackJumpAdditive {nodes [Output Pose() {layerGroup=Item Override}]}
 ItemJetPackStartAdditive {nodes [Output Pose() {layerGroup=Item Override}]}
 ItemJumpLoopAdditive {nodes [
   Input Pose() {name=InPoseItemJumpLoopAdditive}
   Output Pose() {layerGroup=Item Override}
  ]}
 ItemJumpOverride {nodes [
   Input Pose() {name=InPoseJump}
   Output Pose() {layerGroup=Item Override}
  ]}
 ItemJumpUpAdditive {nodes [
   Input Pose() {name=InPoseItemJumpUpAdditive}
   Output Pose() {layerGroup=Item Override}
  ]}
 ItemPreAimOffsets {nodes [
   Input Pose() {name=InPosePreAimOffsets}
   Output Pose() {layerGroup=Item Override}
  ]}
 ItemPreIK {nodes [
   Input Pose() {name=InPose_PreIK}
   Output Pose() {layerGroup=Item Override}
  ]}
 ItemRightHandIK {nodes [
   Input Pose() {name=InPosePreRightHandIK}
   Input Pose() {name=InPosePostRightHandIK}
   Output Pose() {layerGroup=Item Override}
  ]}
 ItemSkeletalControl {nodes [
   Input Pose()
   Output Pose() {layerGroup=Item Override}
  ]}
 ItemSwimJumpFallLoop {nodes [Output Pose() {layerGroup=Item Override}]}
 ItemSwimJumpSurfaceEnd {nodes [Output Pose() {layerGroup=Item Override}]}
 ItemSwimJumpSurfaceLoop {nodes [Output Pose() {layerGroup=Item Override}]}
 ItemUpperBody {nodes [
   Input Pose!() -> (Pose, InputParam:Fort Anim Input Adjusted Aim Structure) {name=InPoseUpperBody}
   Output Pose() {layerGroup=Item Override}
  ]}
 ItemZipLineStartAdditive {nodes [Output Pose() {layerGroup=Item Override}]}
 Lean {nodes [
   n1 = Lean!() -> (then, OwningActor:Actor Object Reference, DeltaTime:double=~) {is=title K2Node_FunctionEntry}
   > Set PreviousCharacterYaw(PreviousCharacterYaw:double=Get CharacterYaw() -> (CharacterYaw:double)) -> (Output_Get:double)
   > Set CharacterYaw(CharacterYaw:double=BreakRotator(InRot=Get Actor Rotation!(self:Actor Object Reference=@n1) -> (ReturnValue:Rotator) {is=title K2Node_CallFunction}).Yaw) -> (Output_Get:double)
   > n2 = Set CharacterYawDelta(CharacterYawDelta:double=float - float(A=Get CharacterYaw(), B=Get PreviousCharacterYaw() -> (PreviousCharacterYaw:double))) -> (Output_Get:double)
   n3 = SafeDivide(A=@n2, B=@n1.DeltaTime)
   > n4 = Set LeanAngle(LeanAngle:double=Clamp Angle!(AngleDegrees:double=float / float(A=@n3, B=200), MinAngleDegrees:double=-1, MaxAngleDegrees:double=1) -> (ReturnValue:double)) -> (Output_Get:double)
   > PrintString(InString="To String (Float)"(InDouble=@n4))
   > Return Node()
  ]}
 Lean2 {nodes [
   BreakRotator(InRot=Get Actor Rotation(self=Lean2!() -> (then, OwningActor:Actor Object Reference, DeltaTime:double=~) {is=title K2Node_FunctionEntry}))
   FInterpTo(Current=0, Target=0, DeltaTime=0, InterpSpeed=10)
  ]}
 Lean3 {nodes [
   n1 = Lean3!() -> (then, OwningActor:Actor Object Reference, CurrentLean:double=~, DeltaTime:double=~, CurrentActorRotation:double=~) {is=title K2Node_FunctionEntry}
   n2 = Get Actor Rotation!(self=@n1) -> (ReturnValue_Roll:float, ReturnValue_Pitch:float, ReturnValue_Yaw:float)
   n3 = float * float(A=float - float(A=@n2.ReturnValue_Yaw, B=@n1.CurrentActorRotation) {is=title K2Node_CallFunction}, B=5)
   > Return Node!(
    execute
    NewActorRotation:double=@n2.ReturnValue_Yaw
    NewLean:double=FInterpTo(Current=@n1.CurrentLean, Target="Clamp (Float)"(Value=@n3, Min=-90, Max=90), DeltaTime=@n1.DeltaTime, InterpSpeed=3)
   )
  ]}
 NewFunction {nodes [NewFunction() {is=title K2Node_FunctionEntry}]}
]
variables [
 Actor {type=Actor, default=null}
 ActorRotation {type=number}
 CharacterYaw {type=number}
 CharacterYawDelta {type=number}
 DeltaTime {type=number}
 LeanAngle {type=number}
 PreviousCharacterYaw {type=number}
 Target {type=SkeletalMeshComponent}
 WheelRotation {type=number}
]
